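# Physics asset for the F_MED_Scuba_01 female medium player body cosmetic, covering its dynamic tube bones and nearby upper-body bones.
# Three small spheres sit on the dyn_tube bones, with capsules on the right clavicle and spine_05, the latter zero-length.
# Each tube body is constrained directly to the right clavicle rather than chained, with swing and twist locked.
# The spine_05 body has no constraint.
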PhysicsAsset /BRCosmetics/Characters/Player/Female/Medium/Bodies/F_MED_Scuba_01/Meshes/F_MED_Scuba_01_Physics
bodies {
 clavicle_r [{shapeType=Capsule, length=9.75, radius=4.5, center=<-5.8682 1.3777 0.7467>, rotation=<rot 79.638 -12.2497 -15.0591>}]
 dyn_tube_01 [{shapeType=Sphere, radius=3}]
 dyn_tube_02 [{shapeType=Sphere, radius=3}]
 dyn_tube_03 [{shapeType=Sphere, radius=3}]
 spine_05 [{shapeType=Capsule, length=0, radius=5}]
}
constraints [
 {
  bone1Name=clavicle_r
  bone2Name=dyn_tube_02
  swing1LimitDegrees=10
  swing1Motion=Locked
  swing2LimitDegrees=10
  swing2Motion=Locked
  twistLimitDegrees=10
  twistMotion=Locked
 }
 {
  bone1Name=clavicle_r
  bone2Name=dyn_tube_01
  swing1LimitDegrees=10
  swing1Motion=Locked
  swing2LimitDegrees=10
  swing2Motion=Locked
  twistLimitDegrees=10
  twistMotion=Locked
 }
 {
  bone1Name=clavicle_r
  bone2Name=dyn_tube_03
  swing1LimitDegrees=10
  swing1Motion=Locked
  swing2LimitDegrees=10
  swing2Motion=Locked
  twistLimitDegrees=10
  twistMotion=Locked
 }
]
