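# Post-process style AnimBP for the female medium Scuba body that copies the attached parent mesh's pose and adds secondary motion.
# Chained AnimDynamics springs sway both flippers, weighted by 2D pawn speed, and a right-arm bandana, then cache the pose.
# When the playlist uses custom character parts, checked at initialization, the output shrinks the head bone to near zero.
# Pose Driver (tube bones) and RigidBody nodes exist but are unconnected.
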AnimBlueprint /BRCosmetics/Characters/Player/Female/Medium/Bodies/F_MED_Scuba_01/Meshes/F_MED_Scuba_01_AnimBP
target_skeleton=&/F_MED_Scuba_01_Skeleton
graphs [
 AnimGraph {nodes [
   Pose Driver - Source: neck_01!(SourcePose:Pose Link Structure) -> (Pose:Pose Link Structure) {
    is=title AnimGraphNode_PoseDriver
    poseAsset=&/F_MED_Scuba_PoseDriver_FullBody_PoseAsset
    onlyDriveBones [{boneName=dyn_PD_tube_01}, {boneName=dyn_PD_tube_02}, {boneName=dyn_PD_tube_03}]
    poseTargets [
     {
      bApplyCustomCurve=false
      bIsHidden=false
      distanceMethod=DefaultMethod
      drivenName=Pose_11
      functionType=DefaultFunction
      targetScale=1
      boneTransforms [
       {targetRotation=<rot 0 0 0>, targetTranslation=<0 0 0>}
       {targetRotation=<rot 0 0 0>, targetTranslation=<0 0 0>}
       {targetRotation=<rot 0 0 0>, targetTranslation=<0 0 0>}
      ]
      customCurve {defaultValue=3.4028234663852886e+38, postInfinityExtrap=RCCE_Constant, preInfinityExtrap=RCCE_Constant, keys []}
      targetRotation=<rot 0 0 0>
     }
     {
      bApplyCustomCurve=false
      bIsHidden=false
      distanceMethod=DefaultMethod
      drivenName=Pose_12
      functionType=DefaultFunction
      targetScale=0.6771
      boneTransforms [
       {targetRotation=<rot 6.8098 5.7456 -20.0892>, targetTranslation=<11.2977 0.1552 -0.0022>}
       {targetRotation=<rot 0.7686 -4.2126 -5.9204>, targetTranslation=<4.5542 0.0017 0.0004>}
       {targetRotation=<rot -1.4438 -18.7438 -9.9751>, targetTranslation=<5.3827 -0.0014 0.0003>}
      ]
      customCurve {defaultValue=3.4028234663852886e+38, postInfinityExtrap=RCCE_Constant, preInfinityExtrap=RCCE_Constant, keys []}
      targetRotation=<rot 0 0 0>
     }
     {
      bApplyCustomCurve=false
      bIsHidden=false
      distanceMethod=DefaultMethod
      drivenName=Pose_13
      functionType=DefaultFunction
      targetScale=0.9787
      boneTransforms [
       {targetRotation=<rot 15.678 4.7284 -24.232>, targetTranslation=<11.2953 0.1538 0.0318>}
       {targetRotation=<rot 2.0908 -4.6175 -6.5792>, targetTranslation=<4.5567 -0.0152 -0.016>}
       {targetRotation=<rot 2.8421 -4.3563 -13.1032>, targetTranslation=<5.3832 0.0076 -0.0026>}
      ]
      customCurve {defaultValue=3.4028234663852886e+38, postInfinityExtrap=RCCE_Constant, preInfinityExtrap=RCCE_Constant, keys []}
      targetRotation=<rot 0 0 0>
     }
     {
      bApplyCustomCurve=false
      bIsHidden=false
      distanceMethod=DefaultMethod
      drivenName=Pose_14
      functionType=DefaultFunction
      targetScale=0.6771
      boneTransforms [
       {targetRotation=<rot 11.2462 -0.0129 7.7417>, targetTranslation=<11.3233 0.1892 -0.0067>}
       {targetRotation=<rot -0.2569 -5.7551 -0.1039>, targetTranslation=<4.5682 0.002 -0.0195>}
       {targetRotation=<rot 0.8098 -28.0353 18.3087>, targetTranslation=<5.4023 0.0118 0>}
      ]
      customCurve {defaultValue=3.4028234663852886e+38, postInfinityExtrap=RCCE_Constant, preInfinityExtrap=RCCE_Constant, keys []}
      targetRotation=<rot 0 0 0>
     }
    ]
    rBFParams {radius=8.9487}
    sourceBones [{boneName=neck_01}, {boneName=neck_02}, {boneName=head}]
   }
   RigidBody() {
    overridePhysicsAsset=&/F_MED_Scuba_01_Physics
    alphaBoolBlend {blendOption=Linear}
    overrideWorldGravity=<1 1 1>
   }
   n1 = "Transform (Modify) Bone - Bone: head"!(ComponentPose=Local To Component(LocalPose=Use cached pose 'PreScale Cache'()), Alpha=1) -> (Pose) {scaleMode=BMM_Replace, alphaBoolBlend {blendOption=Linear}, boneToModify {boneName=head}, scale=<0.01 0.01 0.01>}
   Output Pose(Result=Blend Poses by bool(
     BlendPose_0=Component To Local(ComponentPose=@n1)
     BlendPose_1=Use cached pose 'PreScale Cache'()
     BlendTime_0=0
     BlendTime_1=0
     bActiveValue=Get GetPlayerListUsesCustomCharacterParts() -> (GetPlayerListUsesCustomCharacterParts:bool)
    ) {blendType=Linear}) {layerGroup=None}
   n2 = MapRangeClamped(Value=Get PawnSpeed2D(), InRangeA=0, InRangeB=100, OutRangeA=.2)
   n3 = Anim Dynamics - Bone: dyn_flipper_l_01(ComponentPose=Local To Component(LocalPose=Copy Pose From Mesh() {bUseAttachedParent=true}), Alpha=@n2, ExternalForce="0,0,0") {
    alpha=0.9
    angularDampingOverride=0.98
    angularSpringConstant=3000
    bAngularSpring=true
    bGravityOverrideInSimSpace=true
    bOverrideAngularDamping=true
    numSolverIterationsPostUpdate=4
    numSolverIterationsPreUpdate=16
    simulationSpace=Actor
    alphaBoolBlend {blendOption=Linear}
    boundBone {boneName=dyn_flipper_l_01}
    physicsBodyDefinitions [{
      sphereCollisionRadius=0
      boundBone {boneName=dyn_flipper_l_01}
      boxExtents=<25 20 10>
      constraintSetup {angularTargetAxis=AxisZ, twistAxis=AxisY, angularLimitsMax=<60 0 0>, angularLimitsMin=<-60 0 0>, angularTarget=<0 0 1>}
      localJointOffset=<0 0 0>
     }]
   }
   n4 = Anim Dynamics - Bone: dyn_flipper_r_01(ComponentPose=@n3, Alpha=@n2, ExternalForce="0,0,0") {
    alpha=0.9
    angularDampingOverride=0.98
    angularSpringConstant=3000
    bAngularSpring=true
    bGravityOverrideInSimSpace=true
    bOverrideAngularDamping=true
    numSolverIterationsPostUpdate=4
    numSolverIterationsPreUpdate=16
    simulationSpace=Actor
    alphaBoolBlend {blendOption=Linear}
    boundBone {boneName=dyn_flipper_r_01}
    physicsBodyDefinitions [{
      sphereCollisionRadius=0
      boundBone {boneName=dyn_flipper_r_01}
      boxExtents=<25 20 10>
      constraintSetup {angularTargetAxis=AxisZ, twistAxis=AxisY, angularLimitsMax=<60 0 0>, angularLimitsMin=<-60 0 0>, angularTarget=<0 0 1>}
      localJointOffset=<0 0 0>
     }]
   }
   n5 = Anim Dynamics - Bone: dyn_arm_r_bandana(ComponentPose=@n4, Alpha=0.75, ExternalForce="0,0,0") {
    alpha=0.9
    angularSpringConstant=1500
    bAngularSpring=true
    bGravityOverrideInSimSpace=true
    numSolverIterationsPostUpdate=4
    numSolverIterationsPreUpdate=16
    simulationSpace=Actor
    alphaBoolBlend {blendOption=Linear}
    boundBone {boneName=dyn_arm_r_bandana}
    physicsBodyDefinitions [{
      sphereCollisionRadius=0
      boundBone {boneName=dyn_arm_r_bandana}
      boxExtents=<15 5 5>
      constraintSetup {angularTargetAxis=AxisZ, twistAxis=AxisY, angularLimitsMax=<15 0 15>, angularLimitsMin=<-15 0 -15>, angularTarget=<0 0 1>}
      localJointOffset=<0 0 0>
     }]
   }
   Save cached pose 'PreScale Cache'(Pose=Component To Local(ComponentPose=@n5)) {cacheName=PreScale Cache}
  ]}
 EventGraph {nodes [
   Event BlueprintInitializeAnimation()
   > Set GetPlayerListUsesCustomCharacterParts(GetPlayerListUsesCustomCharacterParts:bool=GetPlaylistUsesCustomCharacterParts!() -> (ReturnValue:bool)) -> (Output_Get:bool)
   Event BlueprintUpdateAnimation()
  ]}
]
variables [GetPlayerListUsesCustomCharacterParts {type=boolean}]
